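# Editor utility widget that converts layered-material static meshes into multi-section meshes and regenerates their LODs, working through a local asset collection.
# Populate fills the collection with layer-UV meshes from Blueprint actor descriptors intersecting the selection bounds.
# Split uses Geometry Script to turn UV1 islands into polygroups, setting material IDs from floored UV X.
# Both split and LOD-only runs use metadata tags to skip converted meshes; MIC generation is an empty stub.
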EditorUtilityWidgetBlueprint /Game/NeverCook/EditorUtilities/GeoscriptTools/LayerSectionConverter/EUW_LayerSectionConverter2
graphs [
 AddAssetsToCollection {nodes [
   n1 = AddAssetsToCollection!() -> (then, ForceEmpty:bool=~, AssetPtrs:Array of Object References) {is=title K2Node_FunctionEntry}
   > n2 = Sequence()
   > Branch(Condition=Get ForceEmpty() -> (ForceEmpty:bool))
   n3 = Get CollectionManagerScriptingSubsystem()
   > CreateOrEmptyCollection!(
    execute
    self:Collection Manager Subsystem Object Reference=@n3
    Container:Collection Container Source Structure=GetBaseGameCollectionContainer!(self:Collection Manager Subsystem Object Reference=@n3) -> (ReturnValue:Collection Container Source Structure)
    Collection:Name=Get CollectionName() -> (CollectionName:Name=None)
    ShareType:ECollectionScriptingShareType Enum=Local
   ) -> (then, OutNewOrEmptyCollection:Collection Structure, ReturnValue:bool)
   n4 = Get CollectionManagerScriptingSubsystem()
   AddAssetPtrsToCollection!(
    execute=@n2.then_1
    self:Collection Manager Subsystem Object Reference=@n4
    Collection:"Collection Structure (by ref)"=Make Collection!(
     Container:Name=GetBaseGameCollectionContainer!(self=@n4) -> (ReturnValue_Name:Name, ReturnValue_Title:Text)
     Name:Name=Get CollectionName() -> (CollectionName=None)
     ShareType:ECollectionScriptingShareType Enum=Local
    ) -> (CollectionScriptingRef:Collection Structure) {is=title K2Node_MakeStruct}
    AssetPtrs:Array of Object References=@n1.AssetPtrs
   ) -> (then, ReturnValue:bool)
  ]}
 DoesMeshHaveLayerUVs {nodes [
   n1 = DoesMeshHaveLayerUVs!() -> (then, TargetMesh:Dynamic Mesh Object Reference) {is=title K2Node_FunctionEntry}
   > n2 = GetUVSetBoundingBox!(execute, TargetMesh:Dynamic Mesh Object Reference=@n1, UVSetIndex:int=1) -> (then, bIsValidUVSet:bool, bUVSetIsEmpty:bool, ReturnValue:Box 2D Structure)
   n3 = BreakVector2D(InVec=Break Box 2D!(Box2D:"Box 2D Structure (by ref)"=@n2.ReturnValue) -> (Min:Vector 2D Structure, Max:Vector 2D Structure, bIsValid:bool=~) {is=title K2Node_BreakStruct}.Max)
   > Return Node!(execute, ReturnValue:bool=AND Boolean!(A=@n2, B=NOT Boolean(A=@n2.bUVSetIsEmpty), C:bool="float > float"(A=@n3, B=1)) -> (ReturnValue))
  ]}
 EventGraph {nodes [
   "On Clicked (btnSplitCollection)"!() -> (OutputDelegate, then, widget:User Widget Object Reference, event:Pointer Event Structure) {is=title K2Node_ComponentBoundEvent}
   > TrySplitCollection!(execute, self, RunMeshSplit:bool=true) -> (then)
   "On Clicked (btnFindAssets)"!() -> (OutputDelegate, then, widget:User Widget Object Reference, event:Pointer Event Structure) {is=title K2Node_ComponentBoundEvent}
   > n1 = Show Message Dialog(
    Title=MakeLiteralText(Value="LOCTEXT(\"Gather Assets\")")
    Message=MakeLiteralText(Value="LOCTEXT(\"Do you want to run the asset gather from selected actors\\' bounds?  It will override the collection!\")")
    MessageType=YesNo
    DefaultValue=No
    MessageCategory=Warning
   )
   > n2 = Switch on EAppReturnType(Selection=@n1)
   GatherActorsForCollection(execute=@n2.Yes)
   "On Clicked (btnLODCollection)"!() -> (OutputDelegate, then, widget:User Widget Object Reference, event:Pointer Event Structure) {is=title K2Node_ComponentBoundEvent}
   > TrySplitCollection(RunMeshSplit=false)
  ]}
 GatherActorsForCollection {nodes [
   n1 = GetSelectedActors(execute=@n27.then_0)
   > n2 = GetActorArrayBounds!(execute, Actors:Array of Actor Object References=@n1, bOnlyCollidingComponents:bool=false) -> (then, Center:Vector, BoxExtent:Vector)
   > n3 = GetIntersectingActorDescs!(
    execute
    InBox:"Box Structure (by ref)"=MakeBoxFromCenterExtents!(Center:Vector=@n2, Extents:Vector=@n2.BoxExtent) -> (ReturnValue:Box Structure)
   ) -> (then, OutActorDescs:Array of Actor Desc Structures, ReturnValue:bool)
   > n4 = For Each Loop(Array:Array of Actor Desc Structures=@n3) -> (Array Element:Actor Desc Structure)
   n5 = Break Actor Desc!(ActorDesc:"Actor Desc Structure (by ref)"=@n4) -> (
    Guid:Guid Structure
    NativeClass:Object Class Reference
    Class:Soft Object Path Structure
    Name:Name
    Label:Name
    Bounds:Box Structure
    RuntimeGrid:Name
    bIsSpatiallyLoaded:bool=~
    bActorIsEditorOnly:bool=~
    ActorPackage:Name
    ActorPath:Name
    FolderPath:Name
    DataLayerAssets:Array of Soft Object Path Structures
   ) {is=title K2Node_BreakStruct}
   n6 = To Soft Object Reference(SoftObjectPath=@n5.Class)
   n7 = LoadAsset_Blocking(execute=@n29.then, Asset=@n6)
   n8 = GetBlueprintAsset(Object=@n7)
   n9 = Cast To Class!(Object:Object Reference=@n7, Class:Object Class Reference=&Actor) -> (ReturnValue:Actor Class Reference)
   n10 = GetBlueprintForClass!(execute=@n7.then, Class:Object Class Reference=@n9) -> (True:Exec, False:Exec, ReturnValue:Blueprint Object Reference)
   n11 = PrintString(execute=@n8.then, InString=Append(A="BP ", B=GetDisplayName(Object=@n10)))
   Is Valid(exec=[@n11.then, @n10.True], InputObject=@n10)
   > n12 = GeneratedClass!(execute, BlueprintObj:Blueprint Object Reference=@n10) -> (then, ReturnValue:Object Class Reference)
   > n13 = Cast To BuildingStaticGeo Class!(execute:Exec, Class:Object Class Reference=@n12) -> (then:Exec, CastFailed:Exec, "AsBuilding Static Mesh (Static Geo)":"Building Static Mesh (Static Geo) Class Reference") {is=title K2Node_ClassDynamicCast}
   n14 = Gather Subobject Data For Blueprint(execute=[@n13.then, @n24.then], self=Get SubobjectDataSubsystem(), Context=@n10)
   > n15 = For Each Loop with Break!(Exec:Exec, Array:Array of Subobject Data Handle Structures=@n14, Break:Exec=[@n23.then, @n22.else]) -> (LoopBody:Exec, Array Element:Subobject Data Handle Structure, Array Index:int, Completed:Exec)
   > n16 = Cast To StaticMeshComponent(Object=GetAssociatedObject!(Data:"Subobject Data Structure (by ref)"=GetData(DataHandle=@n15)) -> (ReturnValue:Object Reference))
   > Is Valid Component!(execute:Exec, MeshComponent:Static Mesh Component Object Reference=@n16) -> (true:Exec, false:Exec) {is=title K2Node_MacroInstance}
   > n17 = GetGlobalMeshPool!(execute) -> (then, ReturnValue:Dynamic Mesh Pool Object Reference)
   > n18 = Request and Release Compute Mesh!(Exec:Exec, Mesh Pool:Dynamic Mesh Pool Object Reference=@n17) -> (With Mesh:Exec, Mesh:Dynamic Mesh Object Reference, After Release:Exec) {is=title K2Node_MacroInstance}
   n19 = Get StaticMesh(self=@n16)
   > n20 = Copy Mesh From Static Mesh!(
    execute
    FromStaticMeshAsset:Static Mesh Object Reference=@n19
    ToDynamicMesh:Dynamic Mesh Object Reference=@n18
    AssetOptions:Geometry Script Copy Mesh From Asset Options Structure
    RequestedLOD:Geometry Script Mesh Read LOD Structure
    bUseSectionMaterials:bool=true
   ) -> (Failure:Exec, Success:Exec, ReturnValue:Dynamic Mesh Object Reference)
   n21 = DoesMeshHaveLayerUVs!(execute=@n20.Success, self, TargetMesh:Dynamic Mesh Object Reference=@n20) -> (then, ReturnValue:bool) {is=title K2Node_CallFunction}
   > n22 = Branch(Condition=@n21)
   > n23 = Add!(
    execute
    TargetSet:Set of Static Mesh Object References=Get l_meshes() -> (l_meshes:Set of Static Mesh Object References)
    NewItem:"Static Mesh Object Reference (by ref)"=@n19
   ) -> (then) {is=title K2Node_CallFunction}
   n24 = Cast To BuildingSMActor Class!(execute:Exec=@n13.CastFailed, Class:Object Class Reference=@n12) -> (then:Exec, CastFailed:Exec, AsBuilding Static Mesh:Building Static Mesh Class Reference) {is=title K2Node_ClassDynamicCast}
   n25 = Reroute Node(InputPin:String) -> (OutputPin:String)
   n26 = GetAssetsByPackageName!(
    execute
    self:Asset Registry Interface=GetAssetRegistry()
    PackageName:Name=String To Name(InString=Split(SourceString=@n25, InStr=., SearchDir=FromStart))
    bIncludeOnlyOnDiskAssets:bool=false
    bSkipARFilteredAssets:bool=true
   ) -> (then, OutAssetData:Array of Asset Data Structures, ReturnValue:bool)
   > Add(
    TargetArray:Array of Asset Data Structures=Get l_assetData() -> (l_assetData:Array of Asset Data Structures)
    NewItem:"Asset Data Structure (by ref)"="Get (a copy)"(Array:Array of Asset Data Structures=@n26) -> (Output:Asset Data Structure)
   ) {is=title K2Node_CallArrayFunction}
   GatherActorsForCollection() {is=title K2Node_FunctionEntry}
   > n27 = Sequence()
   n28 = To Array!(execute=@n27.then_1, A:Set of Static Mesh Object References=Get l_meshes()) -> (then, Result:Array of Static Mesh Object References)
   > AddAssetsToCollection!(execute, self, ForceEmpty:bool=true, AssetPtrs:Array of Object References=@n28) -> (then) {is=title K2Node_CallFunction}
   Contains(
    SearchIn="To String (SoftObjectReference)"!(SoftObjectReference:"Soft Object Reference (by ref)"=@n6) -> (ReturnValue:String)
    Substring=BP_FriedNexus_Quarterwall_A_A_C
   )
   LeftChop(SourceString=@n25, Count=2)
   LoadClassAsset_Blocking(AssetClass=To Soft Class Reference())
   Get Class Display Name(Class=@n9)
   PrintString(InString=GetDisplayName(Object=@n10))
   GetDisplayName(Object=@n7)
   Branch(Condition=EndsWith(SourceString=@n25, InSuffix=_C))
   n29 = Branch(execute=@n4.LoopBody, Condition=@n5.bIsSpatiallyLoaded)
  ]}
 GeoScript_GenerateMICs {nodes [
   GeoScript_GenerateMICs!() -> (then, InStaticMesh:Static Mesh Object Reference) {is=title K2Node_FunctionEntry}
   > Sequence()
  ]}
 GeoScript_SplitMesh {nodes [
   GeoScript_SplitMesh!() -> (then, InStaticMesh:Static Mesh Object Reference, RunMeshSplit:bool=~) {is=title K2Node_FunctionEntry}
   > n1 = GetGlobalMeshPool()
   > n2 = Request and Release Compute Mesh(Mesh Pool=@n1)
   > Set l_dynamicMesh(l_dynamicMesh:Dynamic Mesh Object Reference=@n2) -> (Output_Get:Dynamic Mesh Object Reference)
   > n3 = Copy Mesh From Static Mesh(
    FromStaticMeshAsset=Get InStaticMesh() -> (InStaticMesh:Static Mesh Object Reference)
    ToDynamicMesh=Get l_dynamicMesh() -> (l_dynamicMesh:Dynamic Mesh Object Reference)
    bUseSectionMaterials=true
   )
   n4 = Sequence!(execute=@n3.Success) -> (then_0, then_1, then_2:Exec, then_3:Exec, then_4:Exec)
   > n5 = DoesMeshHaveLayerUVs(TargetMesh=Get l_dynamicMesh())
   > n6 = Branch(Condition=@n5)
   Return Node!(execute=@n6.else, Success:bool=false)
   n7 = Get StaticMaterials!(self:Static Mesh Object Reference=Get InStaticMesh()) -> (StaticMaterials:Array of Static Material Structures)
   Set l_currentMaterial(
    execute=@n4.then_1
    l_currentMaterial:Material Interface Object Reference=Break Static Material!(StaticMaterial:"Static Material Structure (by ref)"="Get (a copy)"(Array:Array of Static Material Structures=@n7) -> (Output:Static Material Structure)) -> (
     MaterialInterface:Material Interface Object Reference
     MaterialSlotName:Name
     UVChannelData:Mesh UVChannel Info Structure
     OverlayMaterialInterface:Material Interface Object Reference
    ) {is=title K2Node_BreakStruct}
   ) -> (Output_Get:Material Interface Object Reference)
   Branch(execute=@n4.then_2, Condition=Get RunMeshSplit() -> (RunMeshSplit:bool))
   > n8 = Convert UV Islands To PolyGroups!(execute, TargetMesh:Dynamic Mesh Object Reference=Get l_dynamicMesh(), GroupLayer:Geometry Script Group Layer Structure, UVLayer:int=1) -> (then, ReturnValue:Dynamic Mesh Object Reference)
   n9 = Get l_polygroups() -> (l_polygroups:Index List Structure)
   > n10 = Get PolyGroup IDs In Mesh!(
    execute
    TargetMesh:Dynamic Mesh Object Reference=@n8
    GroupLayer:Geometry Script Group Layer Structure
    PolygroupIDsOut:"Index List Structure (by ref)"=@n9
   ) -> (then, ReturnValue:Dynamic Mesh Object Reference) {is=title K2Node_CallFunction}
   > n11 = ConvertIndexListToArray!(execute, IndexList:Index List Structure=@n9) -> (then, IndexArray:Array of Integers)
   > n12 = For Each Loop(Array:Array of Integers=@n11) -> (Array Element:int)
   > n13 = Get PolyGroup UV Bounding Box!(execute, TargetMesh:Dynamic Mesh Object Reference=@n10, GroupLayer:Geometry Script Group Layer Structure, GroupID:int=@n12, UVChannel:int=1) -> (then, UVBounds:Box 2D Structure, ReturnValue:Dynamic Mesh Object Reference) {is=title K2Node_CallFunction}
   n14 = Break Box 2D(Box2D=@n13) -> (bIsValid=~)
   > Set PolyGroup Material ID!(
    execute
    TargetMesh:Dynamic Mesh Object Reference=@n10
    GroupLayer:Geometry Script Group Layer Structure
    PolygroupID:int=@n12
    MaterialID:int=Floor(A=BreakVector2D(InVec=@n14.Max)) {is=title K2Node_CallFunction}
    bDeferChangeNotifications:bool=false
   ) -> (then, bIsValidPolygroupID:bool, ReturnValue:Dynamic Mesh Object Reference) {is=title K2Node_CallFunction}
   Return Node(execute=@n4.then_4, Success=true)
   n15 = Branch(execute=@n4.then_3, Condition=Get RunMeshSplit())
   > n16 = Sequence!(execute) -> (then_0, then_1, then_2, then_3)
   > n17 = For Loop(LastIndex=GetMaxMaterialID!(TargetMesh:Dynamic Mesh Object Reference=Get l_dynamicMesh()) -> (bHasMaterialIDs:bool, ReturnValue:int=0).ReturnValue)
   > Add(
    TargetArray:Array of Material Interface Object References=Get l_materialList() -> (l_materialList:Array of Material Interface Object References)
    NewItem:"Material Interface Object Reference (by ref)"=Get l_currentMaterial() -> (l_currentMaterial:Material Interface Object Reference)
   )
   n18 = "To String (Integer)"(InInt=@n17)
   > Add(
    TargetArray:Array of Names=Get l_slotNames() -> (l_slotNames:Array of Names)
    NewItem:"Name (by ref)"=String To Name(InString=Append(A=Slot, B=@n18))
   )
   n19 = CopyMeshToStaticMesh!(
    execute=@n16.then_1
    FromDynamicMesh:Dynamic Mesh Object Reference=Get l_dynamicMesh()
    ToStaticMeshAsset:Static Mesh Object Reference=Get InStaticMesh()
    Options:Geometry Script Copy Mesh To Asset Options Structure=Make Geometry Script Copy Mesh to Asset Options!(
     bReplaceMaterials:bool=true
     NewMaterialUsage:ECopyMeshToAssetUseNewMaterialsOptions Enum=ReplaceAssetMaterials
     NewMaterials:Array of Material Interface Object References=Get l_materialList()
     NewMaterialSlotNames:Array of Names=Get l_slotNames()
     bCleanAssignedMaterials:bool=False
     NewNaniteSettings:Mesh Nanite Settings Structure=Make Mesh Nanite Settings!() -> (MeshNaniteSettings:Mesh Nanite Settings Structure) {is=title K2Node_MakeStruct}
    ) -> (GeometryScriptCopyMeshToAssetOptions:Geometry Script Copy Mesh To Asset Options Structure) {is=title K2Node_MakeStruct}
    TargetLOD:Geometry Script Mesh Write LOD Structure
    bUseSectionMaterials:bool=true
   ) -> (Failure:Exec, Success:Exec, ReturnValue:Dynamic Mesh Object Reference)
   RemoveLods!(
    execute=@n19.Success
    self:Static Mesh Editor Subsystem Object Reference=Get StaticMeshEditorSubsystem()
    StaticMesh:Static Mesh Object Reference=Get InStaticMesh()
   ) -> (then, ReturnValue:bool)
   RemoveLods(execute=[@n16.then_2, @n15.else], self=Get StaticMeshEditorSubsystem(), StaticMesh=Get InStaticMesh())
   n20 = Make Static Mesh Reduction Settings!(PercentTriangles:float=0.15, ScreenSize:float=0.05) -> (StaticMeshReductionSettings:Static Mesh Reduction Settings Structure) {is=title K2Node_MakeStruct}
   n21 = Make Static Mesh Reduction Settings(PercentTriangles=0.05, ScreenSize=0.01)
   n22 = Make Static Mesh Reduction Settings(PercentTriangles=1, ScreenSize=1)
   n23 = Make Static Mesh Reduction Settings(PercentTriangles=0.4, ScreenSize=0.5)
   > SetLods!(
    execute
    self:Static Mesh Editor Subsystem Object Reference=Get StaticMeshEditorSubsystem()
    StaticMesh:Static Mesh Object Reference=Get InStaticMesh()
    ReductionOptions:"Static Mesh Reduction Options Structure (by ref)"=Make Static Mesh Reduction Options!(
     bAutoComputeLODScreenSize:bool=True
     ReductionSettings:Array of Static Mesh Reduction Settings Structures=Make Array!(
      "[0]":Static Mesh Reduction Settings Structure=@n22
      "[1]":Static Mesh Reduction Settings Structure=@n23
      "[2]":Static Mesh Reduction Settings Structure=@n20
      "[3]":Static Mesh Reduction Settings Structure=@n21
     ) -> (Array:Array of Static Mesh Reduction Settings Structures)
    ) -> (StaticMeshReductionOptions:Static Mesh Reduction Options Structure) {is=title K2Node_MakeStruct}
   ) -> (then, ReturnValue:int=0)
   n24 = For Each Loop(Exec=@n16.then_3, Array:Array of Static Material Structures=Get StaticMaterials(self=Get InStaticMesh())) -> (Array Element:Static Material Structure)
   Set StaticMaterials!(
    execute=@n24.Completed
    StaticMaterials:Array of Static Material Structures=Get l_mats() -> (l_mats:Array of Static Material Structures)
    self:Static Mesh Object Reference=Get InStaticMesh()
   ) -> (then, Output_Get:Array of Static Material Structures)
   n25 = Set members in Static Material!(
    execute:Exec=@n24.LoopBody
    StructRef:"Static Material Structure (by ref)"=@n24
    MaterialInterface:Material Interface Object Reference=Get l_currentMaterial()
   ) -> (then:Exec, StructOut:"Static Material Structure (by ref)") {is=title K2Node_SetFieldsInStruct}
   > Add(TargetArray:Array of Static Material Structures=Get l_mats(), NewItem:"Static Material Structure (by ref)"=@n25)
  ]}
 GetAssetsFromCollection {nodes [
   GetAssetsFromCollection() {is=title K2Node_FunctionEntry}
   n1 = Get CollectionManagerScriptingSubsystem()
   > n2 = GetAssetsInCollection!(
    execute
    self:Collection Manager Subsystem Object Reference=@n1
    Collection:"Collection Structure (by ref)"=Make Collection(Container=GetBaseGameCollectionContainer(self=@n1), Name=Get CollectionName() -> (CollectionName=None), ShareType=Local)
   ) -> (then, OutAssets:Array of Asset Data Structures, ReturnValue:bool)
   > Branch(Condition=@n2.ReturnValue)
   > n3 = For Each Loop(Array:Array of Asset Data Structures=@n2) -> (Array Element:Asset Data Structure)
   Return Node!(execute=@n3.Completed, Assets:Array of Object References=Get l_assets() -> (l_assets:Array of Object References))
   > Add(TargetArray:Array of Object References=Get l_assets(), NewItem:"Object Reference (by ref)"=GetAsset(InAssetData=@n3))
  ]}
 HasMeshBeenConverted {nodes [
   n1 = Inputs!() -> (execute, InAsset:Object Reference)
   > n2 = GetMetadataTag!(execute, Object:Object Reference=@n1, Tag:Name=Get ConvertedTagName() -> (ConvertedTagName:Name=None)) -> (then, ReturnValue:String)
   > n3 = Branch(Condition="Equal Exactly (String)"(A=@n2, B=True))
   > Outputs!(Yes:Exec, No:Exec=@n3.else)
  ]}
 IsMeshConvertable {nodes [
   n1 = Inputs!() -> (execute, InMesh:Static Mesh Object Reference)
   > n2 = GetMetadataTag(Object=@n1, Tag=Get CanBeConvertedTagName() -> (CanBeConvertedTagName:Name=None))
   > n3 = Branch(Condition="Not Equal Exactly (String)"(A=@n2, B=False))
   > n4 = Sequence!(execute) -> (then_0, then_1)
   n5 = Get StaticMaterials(self=@n1)
   > n6 = Branch(Condition="integer < integer"(A=Length(TargetArray:Array of Static Material Structures=@n5), B=2))
   > n7 = GetBaseMaterial!(execute, self:Material Interface Object Reference=Break Static Material(StaticMaterial="Get (a copy)"(Array=@n5))) -> (then, ReturnValue:Material Object Reference)
   > n8 = Branch(Condition=Contains Item(
     TargetArray:Array of Material Object References=Get BaseMaterialList() -> (BaseMaterialList:Array of Material Object References)
     ItemToFind:"Material Object Reference (by ref)"=@n7
    ))
   n9 = Branch(execute=[@n8.else, @n6.else], Condition=Get SetMetadata() -> (SetMetadata:bool))
   > n10 = SetMetadataTag!(execute, Object:Object Reference=@n1, Tag:Name=Get CanBeConvertedTagName() -> (CanBeConvertedTagName=None), Value:String=False) -> (then)
   Outputs!(Yes=[@n12.then, @n11.else], No=[@n3.else, @n10.then, @n9.else], MeshPassthrough:Object Reference=@n1)
   n11 = Branch(execute=@n4.then_1, Condition=Get SetMetadata())
   > n12 = SetMetadataTag(Object=@n1, Tag=Get CanBeConvertedTagName() -> (CanBeConvertedTagName=None), Value=True)
   GetAssetRegistry()
  ]}
 IsValidComponent {nodes [
   n1 = Inputs!() -> (execute, MeshComponent:Static Mesh Component Object Reference)
   > n2 = Is Valid(InputObject=Get StaticMesh(self=@n1))
   Outputs!(true:Exec=@n7.then, false:Exec=[@n7.else, @n2.Is Not Valid, @n5.Is Not Valid])
   > n3 = GetMaterials!(execute, self:Mesh Component Object Reference=@n1) -> (then, ReturnValue:Array of Material Interface Object References)
   n4 = "Get (a copy)"(Array:Array of Material Interface Object References=@n3) -> (Output:Material Interface Object Reference)
   > n5 = Is Valid(InputObject=@n4)
   > n6 = GetBaseMaterial(self=@n4)
   > n7 = Branch(Condition=Contains Item(TargetArray=Get BaseMaterialList(), ItemToFind=@n6))
  ]}
 TrySplitCollection {nodes [
   n1 = TrySplitCollection!() -> (then, RunMeshSplit:bool=~) {is=title K2Node_FunctionEntry}
   > n2 = GetAssetsFromCollection!(execute, self) -> (then, Assets:Array of Object References) {is=title K2Node_CallFunction}
   > n3 = For Each Loop(Array:Array of Object References=@n2) -> (Array Element:Object Reference)
   SaveLoadedAssets(execute=@n3.Completed, AssetsToSave=Get l_assetsToSave() -> (l_assetsToSave:Array of Object References), bOnlyIfIsDirty=false)
   n4 = Cast To StaticMesh(execute=@n3.LoopBody, Object=@n3)
   > n5 = TrySplitMeshAndSave!(execute, self, InStaticMesh:Static Mesh Object Reference=@n4, RunMeshSplit:bool=@n1) -> (then, DidConversion:bool)
   > Branch(Condition=@n5)
   > n6 = Branch(Condition=Get SetMetadata())
   > n7 = SetMetadataTag(Object=@n4, Tag=Get ConvertedTagName() -> (ConvertedTagName=None), Value=True)
   Add(execute=[@n7.then, @n6.else], TargetArray=Get l_assetsToSave(), NewItem=@n4)
  ]}
 TrySplitMeshAndSave {nodes [
   n1 = TrySplitMeshAndSave!() -> (then, InStaticMesh:Static Mesh Object Reference, RunMeshSplit:bool=~) {is=title K2Node_FunctionEntry}
   > n2 = Is Mesh Convertable!(execute:Exec, InMesh:Static Mesh Object Reference=@n1) -> (Yes:Exec, No:Exec, MeshPassthrough:Object Reference) {is=title K2Node_MacroInstance}
   Return Node!(execute=@n2.No, DidConversion:bool=false)
   > n3 = Has Mesh Been Converted!(execute:Exec, InAsset:Object Reference=@n2) -> (Yes:Exec, No:Exec) {is=title K2Node_MacroInstance}
   n4 = GeoScript_SplitMesh!(execute=@n3.No, self, InStaticMesh:Static Mesh Object Reference=Get InStaticMesh(), RunMeshSplit:bool=@n1.RunMeshSplit) -> (then, Success:bool) {is=title K2Node_CallFunction}
   > n5 = Branch(Condition=@n4)
   > GeoScript_GenerateMICs!(execute, self, InStaticMesh:Static Mesh Object Reference=Get InStaticMesh()) -> (then) {is=title K2Node_CallFunction}
   > Return Node(DidConversion=true)
   Return Node(execute=@n5.else, DidConversion=false)
   Return Node(execute=@n3.Yes, DidConversion=false)
  ]}
]
variables [
 BaseMaterialList {
  type="array<Material>"
  default [
   /Game/Athena/WinterArt/Materials/M_BaseSimple_4Layers
   /Game/Global/Base4Layers_2022/Material/P_FN_Base_2022
   /Game/Packages/Fortress_SharedMaterials/Base_Material/Fortress_BaseMaterial
  ]
 }
 CanBeConvertedTagName {type=string, default=CanBeConverted}
 CollectionName {type=string, default=LayerSectionConversionAssets}
 ConvertedTagName {type=string, default=HasBeenConverted}
 SetMetadata {type=boolean, default=true}
]
widgets {
 info {rootWidgetClass=&VerticalBox, widgetCount=4}
 widgets [
  VerticalBox_63 {widgetClassPath=&VerticalBox, widget=VerticalBox_63}
  btnFindAssets {
   widgetClassPath=&TA_Generic01_Button
   parent=VerticalBox_63
   widget=btnFindAssets
   properties {label=Populate Collection from Actor Selection}
   slot {class=VerticalBoxSlot, properties {padding=<ltrb 0 0 0 10>}}
  }
  btnSplitCollection {
   widgetClassPath=&TA_Generic01_Button
   parent=VerticalBox_63
   widget=btnSplitCollection
   properties {label=Run Split on Collection}
   slot {class=VerticalBoxSlot}
  }
  btnLODCollection {
   widgetClassPath=&TA_Generic01_Button
   parent=VerticalBox_63
   widget=btnLODCollection
   properties {label=Run Just LOD on Collection}
   slot {class=VerticalBoxSlot, properties {padding=<ltrb 0 10 0 0>}}
  }
 ]
}
